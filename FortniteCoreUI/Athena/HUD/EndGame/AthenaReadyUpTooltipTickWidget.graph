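# End-game HUD widget showing a party member's ready-up state as an avatar icon with a shadowed green check tick.
# Construct and Destruct bind and unbind the readiness-changed delegate, and setting a party member resets readiness to false.
# SetReadiness collapses or reveals the widget and, when ready, plays a bottom-pivoted scale-pop animation.
# A PreConstruct call forcing the ready state sits behind a hard-coded false branch.
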WidgetBlueprint /FortniteCoreUI/Athena/HUD/EndGame/AthenaReadyUpTooltipTickWidget
animations [Ready {
  binding_count=1
  length_seconds=0.35
  object=&self:Ready
  sequence {bindings [Overlay_0 {
     id=#1
     possessed_class=Overlay
     tracks [{
       type=MovieScene2DTransformTrack
       sections [{
         type=MovieScene2DTransformSection
         channels [
          Scale.X {type=MovieSceneScriptingFloatChannel, keys [[0,0],[5,1.9388],[7,1]]}
          Scale.Y {type=MovieSceneScriptingFloatChannel, keys [[0,0],[5,1.9388],[7,1]]}
         ]
        }]
      }]
    }]}
 }]
graphs [
 EventGraph {nodes [
   Event Construct()
   n1 = Create Event()
   > Bind Event to On Readiness Changed Delegate(self:Athena Ready Up Tooltip Tick Widget Object Reference, Delegate=@n1) {is=title K2Node_AddDelegate}
   Event PreConstruct()
   > Branch(Condition=false)
   > SetReadiness!(execute, self, bIsReady:bool=true) -> (then)
   Event OnSetPartyMember!() -> (OutputDelegate, then, NewPartyMember:Fort Party Member Object Reference) {is=title K2Node_Event}
   > SetReadiness(bIsReady=false)
   Event Destruct()
   > Unbind Event from On Readiness Changed Delegate(self:Athena Ready Up Tooltip Tick Widget Object Reference, Delegate=@n1) {is=title K2Node_RemoveDelegate}
  ]}
 SetReadiness {nodes [
   n1 = SetReadiness!() -> (then, bIsReady:bool=~) {is=title K2Node_FunctionEntry}
   > SetVisibility(InVisibility=Select(Option 0=Collapsed, Option 1=SelfHitTestInvisible, Index=@n1))
   > Branch(Condition=@n1)
   > PlayAnimation(InAnimation=Get Ready() -> (Ready:Widget Animation Object Reference))
  ]}
]
variables []
widgets {
 info {inheritedWidgetCount=1, parentClass=/Script/FortniteUI.AthenaReadyUpTooltipTickWidget, rootWidgetClass=&Overlay, widgetCount=4}
 widgets [
  Overlay_0 {widgetClassPath=&Overlay, widget=Overlay_0, properties {renderTransformPivot=<0.5 1>}}
  AvatarIconWidget {
   widgetClassPath=/Script/FortniteUI.FortSocialAvatarIcon
   bInherited=true
   bIsVariable=true
   parent=Overlay_0
   widget=AvatarIconWidget
   properties {
    avatarMaterial=/FortniteCoreUI/Athena/HUD/EndGame/Materials/MI_UI_ReadyUpPlayerAvatar
    brush {
     bIsDynamicallyLoaded=false
     drawAs=Image
     imageType=FullColor
     mirroring=NoMirror
     resourceName=None
     resourceObject=&MI_UI_ReadyUpPlayerAvatar
     tiling=NoTile
     imageSize=<64 64>
     margin=<ltrb 0 0 0 0>
     outlineSettings {
      bUseBrushTransparency=false
      roundingType=HalfHeightRadius
      width=0
      color {colorUseRule=UseColor_Specified, specifiedColor=<rgba 0 0 0 0>}
      cornerRadii=<0 0 0 0>
     }
     tintColor {colorUseRule=UseColor_Specified, specifiedColor=<rgba 1 1 1 1>}
     uVRegion {bIsValid=false, max=<0 0>, min=<0 0>}
    }
   }
   slot {class=OverlaySlot, properties {horizontalAlignment=HAlign_Fill, verticalAlignment=VAlign_Fill}}
  }
  Image_Tick_Shadow {
   widgetClassPath=&Image
   bIsVariable=true
   parent=Overlay_0
   widget=Image_Tick_Shadow
   properties {
    brush {
     bIsDynamicallyLoaded=false
     drawAs=Image
     imageType=NoImage
     mirroring=NoMirror
     resourceName=None
     resourceObject=/Game/UI/Foundation/Textures/Icons/Checks/T-Icon-Check-128
     tiling=NoTile
     imageSize=<32 32>
     margin=<ltrb 0 0 0 0>
     outlineSettings {
      bUseBrushTransparency=false
      roundingType=HalfHeightRadius
      width=0
      color {colorUseRule=UseColor_Specified, specifiedColor=<rgba 0 0 0 0>}
      cornerRadii=<0 0 0 0>
     }
     tintColor {colorUseRule=UseColor_Specified, specifiedColor=<rgba 1 1 1 1>}
     uVRegion {bIsValid=false, max=<0 0>, min=<0 0>}
    }
    colorAndOpacity=<rgba 0.0194 0.0194 0.0666 1>
    renderTransform {angle=0, scale=<1 1>, shear=<0 0>, translation=<4 3>}
   }
   slot {class=OverlaySlot, properties {horizontalAlignment=HAlign_Right, verticalAlignment=VAlign_Bottom}}
  }
  Image_Tick {
   widgetClassPath=&Image
   bIsVariable=true
   parent=Overlay_0
   widget=Image_Tick
   properties {
    brush {
     bIsDynamicallyLoaded=false
     drawAs=Image
     imageType=NoImage
     mirroring=NoMirror
     resourceName=None
     resourceObject=&T-Icon-Check-128
     tiling=NoTile
     imageSize=<32 32>
     margin=<ltrb 0 0 0 0>
     outlineSettings {
      bUseBrushTransparency=false
      roundingType=HalfHeightRadius
      width=0
      color {colorUseRule=UseColor_Specified, specifiedColor=<rgba 0 0 0 0>}
      cornerRadii=<0 0 0 0>
     }
     tintColor {colorUseRule=UseColor_Specified, specifiedColor=<rgba 1 1 1 1>}
     uVRegion {bIsValid=false, max=<0 0>, min=<0 0>}
    }
    colorAndOpacity=<rgba 0 1 0.1133 1>
   }
   slot {class=OverlaySlot, properties {horizontalAlignment=HAlign_Right, verticalAlignment=VAlign_Bottom}}
  }
 ]
}
